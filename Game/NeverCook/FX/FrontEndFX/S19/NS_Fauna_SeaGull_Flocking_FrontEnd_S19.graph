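# Niagara GPU effect for a flock of seagulls in the S19 front end.
# BirdFlocker emits short-lived, zero-size particles with curl noise, which Birds read through a particle attribute reader as moving attraction and vortex centers.
# Mesh birds add smoothed wandering wind under capped speed, and send flap amount and a speed-driven counter to material parameters.
# A region mask and lifetime curve set scale, and birds leaving a kill sphere die.
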NiagaraSystem /Game/NeverCook/FX/FrontEndFX/S19/NS_Fauna_SeaGull_Flocking_FrontEnd_S19
dependencies {
 usedDataInterfaces [&NiagaraDataInterfaceCurve, &NiagaraDataInterfaceParticleRead, &NiagaraDataInterfaceDebugDraw]
 usedDynamicInputs []
 usedModules [&SystemState]
}
emitters {
 BirdFlocker {
  simTarget=GPUComputeSim
  emitterUpdateScript [
   &EmitterState()
   &SpawnRate(Use Spawn Probability=false)
  ]
  particleSpawnScript [
   &InitializeParticle(
    Color=<rgba 150 1 1 150>
    Lifetime=2
    Sprite Size Mode=::NewEnumerator3
    Uniform Sprite Size=0
    UsePositionOffset=false
    Write Material Random=true
   )
   &AddVelocity(Scale Added Velocity=<0 0 0>, Velocity=&RandomRangeVector(Minimum=<-2500 -10 0>, Maximum=<2500 10 0>))
   &SphereLocation(
    MaskSpawnGroup=false
    Offset=<0 0 -150>
    Override Seed=false
    Surface Only=false
    UseNonUniformScale=false
    UseOffset=true
    UseSphereOrientation=false
   )
  ]
  particleUpdateScript [
   &ParticleState()
   &CurlNoiseForce(Noise Frequency=2, Noise Strength=25, Pan Noise=true, Pan Noise Field=<0 0 1>)
   &SolveForcesAndVelocity(Clamp Velocity=true, Limit Acceleration=false, Speed Limit=500)
  ]
  rendererClasses [&NiagaraSpriteRendererProperties]
  renderers [{rendererClass=&NiagaraSpriteRendererProperties}]
 }
 Birds {
  simTarget=GPUComputeSim
  emitterUpdateScript [
   &EmitterState()
   &SpawnRate(SpawnRate=5, Use Spawn Probability=false)
  ]
  particleSpawnScript [
   &FN_SetSpawnGroup()
   &InitializeParticle(
    Color=<rgba 2 2 2 1>
    Lifetime=&RandomRangeFloat(Minimum=8, Maximum=10)
    Mesh Scale Mode=::NewEnumerator1
    Mesh Uniform Scale Max=0.21
    Mesh Uniform Scale Min=0.1
    Sprite Size Mode=::NewEnumerator3
    Uniform Sprite Size=500
    UsePositionOffset=false
    Write Material Random=true
   )
   &SphereLocation(
    MaskSpawnGroup=false
    Non Uniform Scale=<3 1 0.6>
    Offset=<-150 -100 -150>
    Override Seed=false
    Sphere Radius=150
    Surface Only=false
    UseNonUniformScale=true
    UseOffset=true
    UseSphereOrientation=false
   )
   &AddVelocity(Velocity=<25 0 0>)
  ]
  particleUpdateScript [
   &ParticleState()
   &self:PositionReader(
    IndexID:NiagaraInt32=&RandomRangeInt(Minimum=0, Maximum=1)
    New Particle Attribute Reader=DataInterface() {propertyValues {EmitterBinding {bindingMode=Other, emitterName=BirdFlocker}}}
   ) {module=PositionReader}
   /SetVariables(Particles.RandomDirection:Vector3f=&Lerp_Vector(
     A=&MakeVector(
      X=&Waveform(
       Global Phase=&Add_Float(A=$Particles.Age, B=&RandomRangeFloat(Minimum=0, Maximum=5))
       "Waveform [1]"=::NewEnumerator8
       "[1] Frequency"=1
       "[1] Amplitude Min/Max"=<-1 1>
       "[2] Phase"=$Emitter.Age
       "[3] Phase"=$Emitter.Age
      )
      Y=&Waveform(
       Global Phase=&Add_Float(A=$Particles.Age, B=&RandomRangeFloat(Minimum=0, Maximum=5))
       "Waveform [1]"=::NewEnumerator8
       "[1] Frequency"=1
       "[1] Amplitude Min/Max"=<-1 1>
       "[2] Phase"=$Emitter.Age
       "[3] Phase"=$Emitter.Age
      )
      Z=0
     )
     B=<1 0 0>
     Alpha=&Waveform(
      Global Phase=&Add_Float(A=$Particles.Age, B=&RandomRangeFloat(Minimum=0, Maximum=5))
      Global Clamp Min=0.4
      "Waveform [1]"=::NewEnumerator8
      "[1] Frequency"=1
      "[1] Amplitude Min/Max"=<0 1>
      "[2] Phase"=$Emitter.Age
      "[3] Phase"=$Emitter.Age
     )
    )) {isSetParameters=true}
   &InterpolateOverTime(
    Rate Of Change=0.5
    Target Vector Value:Vector3f=&NormalizeVector(Vector To Normalize=$Particles.RandomDirection, Safe Normalize=true, Safe Fallback Threshold=0, Safe Fallback Vector=<0 0 0.001>)
    Variable Type=::NewEnumerator2
   ) {module=InterpolateOverTime001}
   &WindForce(
    Air Resistance=0.1
    Wind Speed=&TransformVector(Vector=$Particles.InterpolateOverTime001.InterpolateOverTime_V3_.Moving Average, Destination Space=::Simulation)
    Wind Speed Scale=&RandomRangeFloat(Minimum=4, Maximum=5)
   )
   &PointAttractionForce(
    Attraction Strength=&RandomRangeFloat(Minimum=0.1, Maximum=0.2)
    Attractor Position:Vector3f=$Particles.FishMoverPosition
    Kill Within Radius=false
    Module.Use Falloff=true
    UseAttractorOffset=false
   )
   &VortexForce(
    Origin Pull Amount=15
    UseOriginOffset=false
    Vortex Axis=&Lerp_Vector(
     A=<-0.15 0.15 0>
     B=<0.25 0 0.25>
     Alpha=&FloatFromCurve(
      FloatCurve=DataInterface() {propertyValues {Curve {keys [
          {interpMode=RCIM_Cubic, tangentMode=RCTM_User, time=0, value=0}
          {interpMode=RCIM_Cubic, tangentMode=RCTM_User, time=0.5, value=1}
          {tangentMode=RCTM_User, time=1, value=0}
         ]}}}
      CurveIndex=$System.NormalizedLoopAge
     )
    )
    Vortex Force Amount=10
    Vortex Origin:Vector3f=$Particles.FishMoverPosition
   )
   &SolveForcesAndVelocity(Acceleration Limit=50, Clamp Velocity=true, Limit Acceleration=true, Speed Limit=40)
   /SetVariables(Particles.FlapTarget:NiagaraFloat=/Niagara/DynamicInputs/Integer/V2/Round_Float(Float:NiagaraFloat=&Waveform(
      "Waveform [1]"=::NewEnumerator8
      "[1] Phase"=&Add_Float(A=$System.Age, B=&RandomRangeFloat(Minimum=0, Maximum=2))
      "[1] Frequency"=&RandomRangeFloat(Minimum=0.4, Maximum=1)
      "[1] Amplitude Min/Max"=<0 1>
      "[2] Average"=false
      "[2] Input Position"=&RandomRangeFloat(Minimum=0, Maximum=1)
      "[2] Phase"=$Particles.Age
      "[3] Phase"=$Emitter.Age
     ))) {enabled=false, isSetParameters=true}
   /SetVariables(Particles.FlapTarget=1) {isSetParameters=true}
   &InterpolateOverTime(
    Rate Of Change:NiagaraFloat=&Lerp_Float(A=1, B=1, Alpha=$Particles.FlapTarget)
    Target Float Value:NiagaraFloat=$Particles.FlapTarget
   )
   /SetVariables(Particles.FlapSmoothed:NiagaraFloat=$Particles.InterpolateOverTime.InterpolateOverTime_Float.Moving Average) {isSetParameters=true}
   /Niagara/Modules/Update/Utility/IncrementOverTime(Rate of Change:NiagaraFloat=&Multiply_Float(
     A=&Add_Float(A=0.15, B=$Particles.FlapSmoothed)
     B=&RemapRange(
      Input Value To Remap=&VectorLength(Vector=$Particles.Velocity)
      Input Min=0
      Input Max=5000
      Output Min=1
      Output Max=5
     )
    ))
   &DynamicMaterialParameters(
    Index 0 Param 1=&Multiply_Float(A=$Particles.FlapSmoothed, B=1)
    Index 0 Param 2=&Multiply_Float(A=$Particles.IncrementOverTime.Counter, B=1)
    Module.Param0WriteEnabled=true
    Module.Param1WriteEnabled=true
    Module.Param2WriteEnabled=false
    Module.Param3WriteEnabled=false
   ) {module=DynamicMaterialParameters001}
   &RegionMask(Outer Vector:Vector3f=<0 0 2>, Region Radius=800, UseOffset=false, UseRotation=false, Value Type:ENiagaraRegionMaskValue=::NewEnumerator1)
   /SetVariables(Particles.Scale=$Particles.RegionMask.Vector) {isSetParameters=true}
   &KillParticlesInVolume(Invert Volume=true, Sphere Radius=2500)
   /SetVariables(Particles.NewScale:Vector3f=&Multiply_VectorByFloat(
     Vector=$Particles.Scale
     Float=&FloatFromCurve(
      FloatCurve=DataInterface() {propertyValues {Curve {keys [
          {interpMode=RCIM_Cubic, tangentMode=RCTM_User, time=0, value=0}
          {interpMode=RCIM_Cubic, tangentMode=RCTM_User, time=0.1, value=1}
          {interpMode=RCIM_Cubic, tangentMode=RCTM_User, time=0.8, value=1}
          {tangentMode=RCTM_User, time=1, value=0}
         ]}}}
      Scale Curve=0.1
     )
    )) {isSetParameters=true}
  ]
  rendererClasses [&NiagaraMeshRendererProperties, &NiagaraSpriteRendererProperties]
  renderers [
   {rendererClass=&NiagaraMeshRendererProperties, rendererIndex=0}
   {rendererClass=&NiagaraSpriteRendererProperties, rendererIndex=1}
  ]
 }
}
summary {
 systemName=NS_Fauna_SeaGull_Flocking_FrontEnd_S19
 emitters [
  {
   emitterName=Birds
   simTarget=GPUComputeSim
   rendererClasses [&NiagaraMeshRendererProperties, &NiagaraSpriteRendererProperties]
  }
  {emitterName=BirdFlocker, simTarget=GPUComputeSim, rendererClasses [&NiagaraSpriteRendererProperties]}
 ]
 userVariables [
  User.FishOffset {type=NiagaraFloat, defaultValue=0, description=""}
  User.HideZP {type=NiagaraFloat, defaultValue=0, description=""}
  User.Push {type=NiagaraFloat, defaultValue=0, description=""}
 ]
}
